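# MetaSound patch from the polyphony active voice control example that updates tone generator parameters on a keyed voice.
# Frequency and Gain floats are converted into named voice parameters targeting the sine frequency and gain controls, then merged into one parameter set.
# A Parameters Voice Controller applies that set for the given Key Index to the incoming Voice Control Stream and outputs the resulting stream.
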MetaSoundPatch /MetasoundPolyphonyInternal/Examples/ActiveVoiceControl/MSP_ToneGeneratorController
inputs [
 Frequency {type=Float}
 Gain {type=Float}
 Key Index {type=Int32}
 Voice Control Stream {type=Polyphony::VoiceControlStream}
]
outputs [Voice Control Stream {type=Polyphony::VoiceControlStream}]
pages [Default {
  input_defaults {Frequency=544.72, Gain=0.36, Key Index=0, Voice Control Stream=null}
  nodes [MetasoundPolyphonyInternalParametersVoiceController_0 = MetasoundPolyphonyInternal.ParametersVoiceController(
    Key Index=^Key Index
    Voice Control Stream=^Voice Control Stream
    Voice Parameters=MetasoundPolyphonyInternal.VoiceParametersMerge2Inputs(
     "1: Voice Params"=MetasoundPolyphonyInternal.ConversionFloatToPolyphony::VoiceParameters(In=^Frequency) {also {defaults {Parameter Name=Polyphony.Voice.Control.Sine.Frequency}, version [0,1]}}.Voice Parameters
     "0: Voice Params"=MetasoundPolyphonyInternal.ConversionFloatToPolyphony::VoiceParameters(In=^Gain) {also {defaults {Parameter Name=Polyphony.Voice.Control.Gain.Gain}, version [0,1]}}.Voice Parameters
    ) {also {version [0,1]}}.Voice Params
   ) {also {version [0,1]}}]
  links [{dst="@Voice Control Stream", src=MetasoundPolyphonyInternalParametersVoiceController_0.Voice Control Stream}]
 }]
