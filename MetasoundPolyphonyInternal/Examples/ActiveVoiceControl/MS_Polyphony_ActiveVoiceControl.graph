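# Mono MetaSound source example of active voice control in polyphony, playing up to three tone-generator voices.
# On play, three staggered trigger delays merge into a counter whose value becomes the key index for a TriggerIndexVoiceController.
# Its voice control stream passes through three chained MSP_ToneGeneratorController patches, each keyed by index with exposed frequency and gain inputs.
# A PolyphonyPatchPlayer renders the stream using MSP_ToneGeneratorVoice.
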MetaSoundSource /MetasoundPolyphonyInternal/Examples/ActiveVoiceControl/MS_Polyphony_ActiveVoiceControl
inputs [
 Frequency_0 {type=Float}
 Frequency_1 {type=Float}
 Frequency_2 {type=Float}
 Gain_0 {type=Float}
 Gain_1 {type=Float}
 Gain_2 {type=Float}
 UE.Source.OnPlay {type=Trigger}
]
interfaces [UE.OutputFormat.Mono:1.0, UE.Source:1.1]
outputs ["UE.OutputFormat.Mono.Audio:0" {type=Audio}]
pages [Default {
  input_defaults {Frequency_0=465.5201, Frequency_1=659.965, Frequency_2=659.255, Gain_0=0.496, Gain_1=0.5, Gain_2=0.52, UE.Source.OnPlay=null}
  nodes [
   TriggerDelay_0 = UE.Trigger Delay(In=^UE.Source.OnPlay) {also {defaults {Delay Time=1.5}}}
   TriggerDelay_1 = UE.Trigger Delay(In=^UE.Source.OnPlay)
   TriggerDelay_2 = UE.Trigger Delay(In=^UE.Source.OnPlay) {also {defaults {Delay Time=0.5}}}
   TriggerAnyTriggerAny3_0 = "TriggerAny.Trigger Any (3)"(In 2=@TriggerDelay_0.Out, In 1=@TriggerDelay_1.Out, In 0=@TriggerDelay_2.Out)
   TriggerCounter_0 = UE.Trigger Counter(In=@TriggerAnyTriggerAny3_0.Out)
   n1 = &/MSP_ToneGeneratorController(
    Frequency=^Frequency_0
    Gain=^Gain_0
    Voice Control Stream=MetasoundPolyphonyInternal.TriggerIndexVoiceController(Key Index=Convert.Int32:Float(Float=@TriggerCounter_0.Value).Int32, Trigger=@TriggerCounter_0.On Trigger) {also {version [0,1]}}.Voice Control Stream
   )
   MetasoundPolyphonyInternalPolyphonyPatchPlayer_0 = MetasoundPolyphonyInternal.PolyphonyPatchPlayer:Mono(Voice Control Stream=&MSP_ToneGeneratorController(
     Frequency=^Frequency_2
     Gain=^Gain_2
     Voice Control Stream=&MSP_ToneGeneratorController(Frequency=^Frequency_1, Gain=^Gain_1, Voice Control Stream=@n1.Voice Control Stream) {also {defaults {Key Index=1}}}.Voice Control Stream
    ) {also {defaults {Key Index=2}}}.Voice Control Stream) {also {defaults {MetaSound Asset=&/MSP_ToneGeneratorVoice}, version [0,1]}}
   UE.TriggerRepeat()
  ]
  links [{dst="@UE.OutputFormat.Mono.Audio:0", src=MetasoundPolyphonyInternalPolyphonyPatchPlayer_0.Out Mono}]
 }]
properties {
 assetImportData=&self:AssetImportData
 bLooping=true
 duration=10000
 sampleRate=48000
}
